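# Niagara trail and glow effect for the Alfredo Male glider cosmetic, tinted by the user Emissive Color.
# Two bone-driven anim trail ribbons stream backward and fade over their lifetime, with alpha fed by a system integrator.
# Center and wing glow sprites follow skeletal mesh points, flicker via a looping curve and brighten by day; wing glows also emit lights.
# A SpeedLine emitter, scaled by forward motion or front-end deployment, is disabled.
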NiagaraSystem /BRCosmetics/Gadgets/Assets/VinderTech_GliderChute/Glider_Alfredo_Male/FX/NS_Glider_AlfredoMale_Trail
dependencies {
 usedDataInterfaces [
  &NiagaraDataInterfaceVector2DCurve
  &NiagaraDataInterfaceCurve
  &NiagaraDataInterfaceColorCurve
  &NiagaraDataInterfaceMeshRendererInfo
  &NiagaraDataInterfaceArrayFloat3
 ]
 usedDynamicInputs []
 usedModules [
  &SystemState
  &FN_Integrator
  &self:Set_Trail_Opacity_01
 ]
}
emitters {
 FN_AnimTrail001 {
  emitterSpawnScript [&FN_RibbonIDOffsetUtility()]
  emitterUpdateScript [
   &EmitterState()
   &FN_SpawnAnimTrail(Number of Anim Trails=2)
  ]
  particleSpawnScript [
   &FN_InitializeAnimTrailBoneIndecies()
   &FN_InitializeAnimTrailBoneIndecies(AnimTrail Selection=1, First Bone or Socket Index=2, Second Bone or Socket Index=3) {module=FN_InitializeAnimTrailBoneIndecies001}
   &FN_InitializeAnimTrail(
    Color=$User.Emissive Color
    Position Offset=&FN_FrontEndVector3Modifier(Gameplay=<0 0 0>, Front End=<0 0 0>)
   )
   &AddVelocity(Velocity=&FN_FrontEndVector3Modifier(Gameplay=<0 -900 0>, Front End=<0 -900 0>))
  ]
  particleUpdateScript [
   &SolveForcesAndVelocity(Clamp Velocity=false, Limit Acceleration=false)
   &FN_UpdateAnimTrail(
    Locked Particle Color:LinearColor=&MultiplyLinearColor(
     Linear Color=$Particles.Initial.Color
     Scale Factor=&MakeFloat4FromVectorAndFloat("Vector (RGB)"=<1 1 1>, "Float (Alpha)"=$System.FN_Integrator001.Value)
    )
    Locked Particle RibbonWidth=1
    SkeletalMesh=$User.SkeletalMesh
   ) {module=FN_UpdateAnimTrail001}
   /SetVariables(Particles.RibbonU0Override=$Particles.NormalizedAge) {isSetParameters=true}
   &ScaleColor(
    Color Value To Scale=$Particles.Color
    Scale Alpha=&FN_FrontEndFloatModifier(Gameplay=&Multiply_Float(A=$System.FN_Integrator001.Value, B=$System.TrailOpacity), Front End=1)
    ScaleA=true
    ScaleRGB=false
   ) {enabled=false, module=ScaleColor001}
   &ScaleColor(
    Scale Alpha=&FloatFromCurve(
     FloatCurve=DataInterface() {propertyValues {Curve {keys [{interpMode=RCIM_Cubic, tangentMode=RCTM_User, time=0, value=1}, {interpMode=RCIM_Cubic, tangentMode=RCTM_User, time=0.3, value=0}]}}}
     Scale Curve=0.8
    )
    ScaleA=true
    ScaleRGB=false
   )
  ]
  rendererClasses [&NiagaraRibbonRendererProperties]
  renderers [
   {rendererClass=&NiagaraRibbonRendererProperties}
   {rendererClass=&NiagaraRibbonRendererProperties, rendererIndex=1}
  ]
 }
 Glow_Center {
  emitterUpdateScript [
   &EmitterState(
    Inactive Response=::NewEnumerator1
    Life Cycle Mode=::NewEnumerator1
    Loop Behavior=::NewEnumerator1
    Loop Duration Mode=::NewEnumerator1
    UseLoopDelay=false
   )
   &SpawnBurst_Instantaneous(Use Spawn Probability=false, UseLoopCountLimit=true)
  ]
  particleSpawnScript [&InitializeParticle(Color=$User.Emissive Color, Sprite Size Mode=::NewEnumerator3, Uniform Sprite Size=16, UsePositionOffset=true, Write Material Random=true)]
  particleUpdateScript [
   &SkeletalMeshLocation(
    Bone / Socket Index=4
    MaskSpawnGroup=false
    Mesh Sampling Type=::NewEnumerator5
    Skeletal Mesh=$User.SkeletalMesh
    Socket Sampling Mode=::NewEnumerator5
   )
   &ParticleState(Kill Particles When Lifetime Has Elapsed=false, Let Infinitely Lived Particles Die When Emitter Deactivates=true)
   &ScaleColor(
    Scale Alpha=&FNTimeOfDayScalar(Night Scale=1, Morning Scale=1, Day Scale=2, Evening Scale=1)
    Scale RGB=&VectorFromFloat(Value=&FloatFromCurve(
      FloatCurve=DataInterface() {propertyValues {Curve {keys [<at 0 1>,<at 0.0495 1.0491>,<at 0.0791 0.8714>,<at 0.1561 0.6964>,<at 0.222 1.0327>,<at 0.27 0.8868>,<at 0.3207 0.9188>,<at 0.3738 0.718>,<at 0.4401 0.8429>,<at 0.4803 1.0545>,<at 0.5384 0.9243>,<at 0.5897 0.7018>,<at 0.6556 0.7669>,<at 0.6931 0.9482>,<at 0.7538 1.0328>,<at 0.8218 0.8646>,<at 0.8909 0.8732>,<at 0.9368 0.756>,<at 1 1>]}}}
      CurveIndex=$System.NormalizedLoopAge
     ))
    ScaleA=true
    ScaleRGB=true
   ) {module=ScaleColor002}
  ]
  rendererClasses [&NiagaraSpriteRendererProperties]
  renderers [{rendererClass=&NiagaraSpriteRendererProperties}]
 }
 Glow_Wings {
  emitterUpdateScript [
   &EmitterState(Inactive Response=::NewEnumerator1, Life Cycle Mode=::NewEnumerator1, UseLoopDelay=false)
   &SpawnBurst_Instantaneous(Use Spawn Probability=false, UseLoopCountLimit=true)
   &SpawnBurst_Instantaneous(Spawn Group=1, Use Spawn Probability=false, UseLoopCountLimit=true) {module=SpawnBurst_Instantaneous001}
   &SpawnBurst_Instantaneous(Spawn Group=2, Use Spawn Probability=false, UseLoopCountLimit=true) {module=SpawnBurst_Instantaneous002}
   &SpawnBurst_Instantaneous(Spawn Group=3, Use Spawn Probability=false, UseLoopCountLimit=true) {module=SpawnBurst_Instantaneous003}
  ]
  particleSpawnScript [
   &FN_SetSpawnGroup()
   &InitializeParticle(
    Color=$User.Emissive Color
    Position Offset Coordinate Space=::Local
    Sprite Size=<22 84>
    Sprite Size Mode=::NewEnumerator4
    UsePositionOffset=true
    Write Material Random=true
   )
   &SkeletalMeshLocation(
    Bone / Socket Index=$Emitter.SpawnGroup
    Bone Sampling Mode=::NewEnumerator4
    MaskSpawnGroup=false
    Sampled Position Offset=&Lerp_Vector(A=<3 -4 -19>, B=<-3 -4 -19>, Alpha=&MakeFloatFromInt(Int=$Emitter.SpawnGroup))
    Sampled Position Offset Coordinate Space:ENiagaraSkelMeshTransforms=::NewEnumerator2
    Skeletal Mesh=$User.SkeletalMesh
   ) {enabled=false, module=SkeletalMeshLocation001}
  ]
  particleUpdateScript [
   &ParticleState(
    Kill Particles When Lifetime Has Elapsed=false
    Let Infinitely Lived Particles Die When Emitter Deactivates=true
    Loop Particles Lifetime=true
   )
   &SkeletalMeshLocation(
    Bone / Socket Index=&SpawnGroup()
    Bone Sampling Mode=::NewEnumerator4
    MaskSpawnGroup=false
    Sampled Position Offset=&SelectVectorFromArray(
     Direct Array Index=&SpawnGroup()
     Vector Selection Array=DataInterface() {propertyValues {FloatData [<-2 0 -22>, <2 0 22>, <0 0 -20>, <0 0 20>]}}
    )
    Skeletal Mesh=DataInterface() {propertyValues {
      DefaultMesh=null
      ExcludeBoneName=root
      OptionalComponentClass=null
      PreviewMesh=/BRCosmetics/Gadgets/Assets/VinderTech_GliderChute/Glider_Alfredo_Male/Meshes/Alfredo_Male_Glider
      SkinningMode=SkinOnTheFly
      SoftSourceActor=null
      SourceMode=Default
      UvSetIndex=0
      WholeMeshLOD=-1
      bExcludeBone=false
      bReadDeformedGeometry=true
      bRequireCurrentFrameData=true
      ComponentTags []
      FilteredBones [JNT_R_wing_04, JNT_L_wing_04, JNT_R_wing_02, JNT_L_wing_02]
      FilteredSockets []
      MeshUserParameter {parameter {name=None, typeDefHandle {registeredTypeIndex=1}, varData []}}
      SamplingRegions []
     }}
   )
   &SpriteFacingAndAlignment(
    Sprite Alignment=&RotateVectorByQuaternion(Vector To Rotate=<0 0 1>, Quaternion=$Particles.SkeletalMeshLocation.SampledRotation)
    WriteAlignment=true
    WriteFacing=false
   ) {module=SpriteFacingAndAlignment001}
   &FN_Light_Attributes(
    Color=$User.Emissive Color
    Create Color Attribute=true
    Create Exponent Attribute=true
    Create Position Attribute=true
    Create Radius Attribute=true
    Exponent=4
    Position:NiagaraPosition=$Particles.Position
    Radius=28
    Scale Color=11
   )
   &ScaleColor(
    Scale Alpha=&FNTimeOfDayScalar(Night Scale=1, Morning Scale=1, Day Scale=1.5, Evening Scale=1)
    Scale RGB=&VectorFromFloat(Value=&FloatFromCurve(
      FloatCurve=DataInterface() {propertyValues {Curve {keys [<at 0 1>,<at 0.0495 1.0491>,<at 0.0791 0.8714>,<at 0.1561 0.6964>,<at 0.222 1.0327>,<at 0.27 0.8868>,<at 0.3207 0.9188>,<at 0.3738 0.718>,<at 0.4401 0.8429>,<at 0.4803 1.0545>,<at 0.5384 0.9243>,<at 0.5897 0.7018>,<at 0.6556 0.7669>,<at 0.6931 0.9482>,<at 0.7538 1.0328>,<at 0.8218 0.8646>,<at 0.8909 0.8732>,<at 0.9368 0.756>,<at 1 1>]}}}
      CurveIndex=$System.NormalizedLoopAge
     ))
    ScaleA=true
    ScaleRGB=true
   ) {module=ScaleColor002}
  ]
  rendererClasses [&NiagaraSpriteRendererProperties, &NiagaraLightRendererProperties]
  renderers [
   {rendererClass=&NiagaraSpriteRendererProperties}
   {rendererClass=&NiagaraLightRendererProperties, rendererIndex=1}
   {rendererClass=&NiagaraLightRendererProperties, rendererIndex=2}
  ]
 }
 SpeedLine {
  bEnabled=false
  emitterUpdateScript [
   &EmitterState()
   &SpawnRate(
    SpawnRate=&MakeCustomFloatFromBool(Bool=$User.bIsFrontEndPreview, True Float=50, False Float=90)
    Use Spawn Probability=true
   ) {module=SpawnRate001}
  ]
  particleSpawnScript [
   &InitializeParticle(
    Color=&MakeLinearColorFromVectorAndFloat(
     "Vector (RGB)"=<1 1 1>
     "Float (Alpha)"=&Multiply_Float(
      A=&UniformRangedFloat(Minimum=0.1, Maximum=0.18)
      B=&MakeCustomFloatFromBool(Bool=$User.bIsFrontEndPreview, True Float=0.5, False Float=1)
     )
    )
    Lifetime Max=0.16
    Lifetime Min=0.1
    Lifetime Mode=::NewEnumerator1
    Sprite Size Max=<3 10>
    Sprite Size Min=<2 2>
    Sprite Size Mode=::NewEnumerator2
    UsePositionOffset=false
    Write Material Random=true
   )
   &TorusLocation(
    Align Torus With Axis=true
    Handle Radius=0
    Large Radius:NiagaraFloat=&RandomRangeFloat(Minimum=80, Maximum=150)
    MaskSpawnGroup=false
    Module.Override Randomness=false
    Offset=<0 -180 0>
    Torus Axis=<0 1 0>
    Torus Origin:NiagaraPosition=&Add_Vector(
     A=Linked() {linkedVariable {name=Particles.Position, type=NiagaraPosition}}
     B=&MakeVector(X=0, Y=&MakeCustomFloatFromBool(Bool=$User.bIsFrontEndPreview, True Float=200, False Float=0), Z=0)
    )
   )
   &AddVelocityFromPoint(
    Module.Use Falloff Distance=false
    Origin Offset=<0 -500 0>
    Velocity Strength=&Multiply_Float(A=600, B=&MakeCustomFloatFromBool(Bool=$User.bIsFrontEndPreview, True Float=-1, False Float=1))
   )
   &AddVelocity(Velocity=&Multiply_Vector(
     A=&RandomRangeVector(Minimum=<0 300 0>, Maximum=<0 700 0>)
     B=&MakeVector(X=1, Y=&MakeCustomFloatFromBool(Bool=$User.bIsFrontEndPreview, True Float=-1, False Float=1), Z=1)
    ))
  ]
  particleUpdateScript [
   &SolveForcesAndVelocity(Clamp Velocity=false, Limit Acceleration=false)
   &ParticleState()
   &ScaleSpriteSize(Scale Factor=&Vector2DFromCurve(
     Vector2Curve=DataInterface() {propertyValues {XCurve {keys [<at 0 0.5 RCIM_Cubic>,<at 0.5 1 RCIM_Cubic>]}, YCurve {keys [<at 0 0 RCIM_Cubic>,<at 0.5 1 RCIM_Cubic>]}}}
     ScaleCurve=&Vector2DFromFloat(Value=1)
    ))
   &ScaleSpriteSizeBySpeed(Max Scale Factor=<1 50>, Min Scale Factor=<0.01 0.05>, Velocity Threshold=1)
   &ScaleColor(
    Scale Alpha=&FloatFromCurve(
     FloatCurve=DataInterface() {propertyValues {Curve {keys [
         <at 0 0 RCIM_Cubic>
         <at 0.3759 1.0207>
         {arriveTangent=-2.221, interpMode=RCIM_Cubic, leaveTangent=-2.221, tangentMode=RCTM_User, time=1, value=0}
        ]}}}
     Scale Curve=&MakeCustomFloatFromBool(
      Bool=$User.bIsFrontEndPreview
      True Float=&MakeCustomFloatFromBool(Bool=$User.bIsFullyDeployed, True Float=1, False Float=0)
      False Float=&Lerp_Float(A=0, B=1, Alpha=$User.ForwardDot)
     )
    )
    Scale RGB=&VectorFromFloat(Value=1)
    ScaleA=true
    ScaleRGB=true
   )
  ]
  rendererClasses [&NiagaraSpriteRendererProperties]
  renderers [{rendererClass=&NiagaraSpriteRendererProperties}]
 }
}
summary {
 systemName=NS_Glider_AlfredoMale_Trail
 emitters [
  {bEnabled=false, emitterName=SpeedLine, rendererClasses [&NiagaraSpriteRendererProperties]}
  {emitterName=Glow_Center, rendererClasses [&NiagaraSpriteRendererProperties]}
  {emitterName=Glow_Wings, rendererClasses [&NiagaraSpriteRendererProperties, &NiagaraLightRendererProperties]}
  {emitterName=FN_AnimTrail001, rendererClasses [&NiagaraRibbonRendererProperties]}
 ]
 userVariables [
  User.bIsFrontEnd {type=NiagaraBool}
  User.bIsFrontEndPreview {type=NiagaraBool}
  User.bIsFullyDeployed {type=NiagaraBool}
  User.Emissive Color {type=LinearColor, description="", defaultValue {struct=LinearColor, value=<rgba 0 0.498 1 1>}}
  User.ForwardDot {type=NiagaraFloat, defaultValue=0, description=""}
  User.RibbonIDOffset {type=NiagaraInt32}
  User.SkeletalMesh {
   type=NiagaraDataInterfaceSkeletalMesh
   defaultValue {
    struct=NiagaraExt_VariableValue_DataInterface
    value {dataInterface=&self:NiagaraDataInterfaceSkeletalMesh_0}
   }
  }
 ]
}
