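# Niagara deploy effect for the Alfredo male glider, built from three sprite emitters tinted by User.Emissive Color.
# Beams waits briefly, then bursts speed-stretched sprites along four cone axes picked by spawn group, shaped by curl noise, rising drag and inherited source movement.
# Round_Flare and Wide_Flare spawn camera-offset glow sprites that track User.SkeletalMesh, animating dynamic material parameters and brightness over their short lives.
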NiagaraSystem /BRCosmetics/Gadgets/Assets/VinderTech_GliderChute/Glider_Alfredo_Male/FX/NS_Glider_Alfredo_Male_Deploy
dependencies {
 usedDataInterfaces [&NiagaraDataInterfaceCurve, &NiagaraDataInterfaceArrayFloat3, &NiagaraDataInterfaceColorCurve]
 usedDynamicInputs []
 usedModules [&SystemState]
}
emitters {
 Beams {
  emitterUpdateScript [
   &EmitterState(
    Inactive Response=::NewEnumerator1
    Life Cycle Mode=::NewEnumerator1
    Loop Behavior=::NewEnumerator1
    Loop Delay=0.5
    Loop Duration=1
    UseLoopDelay=true
   )
   &SpawnBurst_Instantaneous(Spawn Count=15, Use Spawn Probability=false)
  ]
  particleSpawnScript [
   &FN_SetSpawnGroup()
   &InitializeParticle(
    Color=&MultiplyLinearColor(Linear Color=$User.Emissive Color, Scale Factor=<1 1 1 1>)
    Lifetime Max=0.85
    Lifetime Min=0.6
    Lifetime Mode=::NewEnumerator1
    Sprite Size Max=<5 5>
    Sprite Size Min=<3 3>
    Sprite Size Mode=::NewEnumerator2
    UsePositionOffset=false
    Write Material Random=true
   ) {module=InitializeParticle002}
   &AddVelocityInCone(
    Cone Angle=50
    Cone Axis=&SelectVectorFromArray(
     Direct Array Index=$Particles.SpawnGroup
     Vector Selection Array=DataInterface() {propertyValues {FloatData [<1 -0.5 0.3>, <1 -0.5 -0.5>, <-1 -0.8 0.2>, <-1 -0.8 -0.5>]}}
    )
    Module.Use Velocity Falloff On Cone Axis=false
    Velocity Strength=1300
   ) {module=AddVelocityInCone001}
  ]
  particleUpdateScript [
   &ParticleState()
   &InheritSourceMovement(Applied Position Scale=<0.9 0.9 0.9>, ApplyForce=false, ApplyPosition=true, ApplyPositionFalloff=false)
   /Niagara/Modules/Update/Forces/CurlNoiseForce(Noise Frequency=0.3, Noise Strength=400, Pan Noise=false)
   &ScaleColor(
    Scale Alpha=&FloatFromCurve(FloatCurve=DataInterface() {propertyValues {Curve {keys [<at 0 1>,<at 1 0>]}}})
    Scale RGB=&Multiply_Vector(
     A=&MakeVectorFromLinearColorRGB(Color=&ColorFromCurve(ColorCurve=DataInterface() {propertyValues {
         AlphaCurve {keys [<at 0.1151 1>,<at 0.9977 1 RCIM_Cubic>]}
         BlueCurve {keys [<at 0.1176 2>,<at 0.2534 1>,<at 0.4815 2>,<at 0.7516 1>,<at 0.9366 2>]}
         GreenCurve {keys [<at 0.1176 1.8257>,<at 0.2534 0.936>,<at 0.4815 0.8319>,<at 0.7516 0.8964>,<at 0.9366 1.5736>]}
         RedCurve {keys [<at 0.1176 1.7389>,<at 0.2534 0.852>,<at 0.4815 0.2944>,<at 0.7516 0.8607>,<at 0.9366 1.6019>]}
        }}))
     B=&VectorFromFloat(Value=11)
    )
    ScaleA=true
    ScaleRGB=true
   )
   &Drag(
    Drag=&FloatFromCurve(
     FloatCurve=DataInterface() {propertyValues {Curve {keys [
         {arriveTangent=4.1785, interpMode=RCIM_Cubic, leaveTangent=4.1785, tangentMode=RCTM_User, time=0, value=0}
         {arriveTangent=-0.0405, interpMode=RCIM_Cubic, leaveTangent=-0.0405, tangentMode=RCTM_User, time=1, value=1}
        ]}}}
     Scale Curve=13
    )
    Use Linear Drag=true
    Use Rotational Drag=false
   )
   &CameraOffset(Camera Offset Amount=-10)
   &SolveForcesAndVelocity(Clamp Velocity=false, Limit Acceleration=false)
   &ScaleSpriteSizeBySpeed(Max Scale Factor=<0.75 10>)
   &ScaleSpriteSize(Scale Factor=&Vector2DFromFloat(Value=&FloatFromCurve(FloatCurve=DataInterface() {propertyValues {Curve {keys [
          {tangentMode=RCTM_User, time=0, value=1}
          {arriveTangent=-0.0405, interpMode=RCIM_Cubic, leaveTangent=-0.0405, tangentMode=RCTM_User, time=0.7009, value=1}
          {arriveTangent=-8.0089, interpMode=RCIM_Cubic, leaveTangent=-8.0089, tangentMode=RCTM_User, time=1, value=0}
         ]}}}))) {module=ScaleSpriteSize001}
  ]
  rendererClasses [&NiagaraSpriteRendererProperties]
  renderers [{rendererClass=&NiagaraSpriteRendererProperties}]
 }
 Round_Flare {
  emitterUpdateScript [
   &EmitterState()
   &SpawnBurst_Instantaneous(Use Spawn Probability=false, UseLoopCountLimit=false)
  ]
  particleSpawnScript [
   &InitializeParticle(
    Color=$User.Emissive Color
    Lifetime=0.7
    Position Offset=<0 20 0>
    Sprite Size Mode=::NewEnumerator3
    Uniform Sprite Size=400
    UsePositionOffset=true
    Write Material Random=true
   )
   &CameraOffset(Camera Offset Amount=3)
  ]
  particleUpdateScript [
   &SkeletalMeshLocation(MaskSpawnGroup=false, Mesh Sampling Type=::NewEnumerator5, Skeletal Mesh=$User.SkeletalMesh, Socket Sampling Mode=::NewEnumerator5)
   &ParticleState()
   &ScaleColor(
    Scale Alpha=&FloatFromCurve(FloatCurve=DataInterface() {propertyValues {Curve {keys [<at 0.9 1 RCIM_Cubic>,<at 1 0 RCIM_Cubic>]}}})
    Scale RGB=&VectorFromFloat(Value=&FloatFromCurve(FloatCurve=DataInterface() {propertyValues {Curve {keys [<at 0 5 RCIM_Cubic>,<at 0.75 5 RCIM_Constant>,<at 0.9 6 RCIM_Constant>]}}}))
    ScaleA=true
    ScaleRGB=true
   )
   &DynamicMaterialParameters(
    Index 0 Param 1=&FloatFromCurve(FloatCurve=DataInterface() {propertyValues {Curve {keys [<at 0 0.9>,<at 0.5 0.05>]}}}, Scale Curve=0.5)
    Index 0 Param 2=0.25
    Index 0 Param 3=1
    Index 0 Param 4=1
    Module.Param0WriteEnabled=true
    Module.Param1WriteEnabled=true
    Module.Param2WriteEnabled=true
    Module.Param3WriteEnabled=true
   )
   &ScaleSpriteSize(
    Uniform Curve Scale=0.5
    Uniform Curve Sprite Scale=DataInterface() {propertyValues {Curve {keys [{tangentMode=RCTM_User, time=0.9, value=1}, {tangentMode=RCTM_User, time=1, value=2}]}}}
   )
  ]
  rendererClasses [&NiagaraSpriteRendererProperties]
  renderers [{rendererClass=&NiagaraSpriteRendererProperties}]
 }
 Wide_Flare {
  emitterUpdateScript [
   &EmitterState()
   &SpawnBurst_Instantaneous(Use Spawn Probability=false, UseLoopCountLimit=false)
  ]
  particleSpawnScript [
   &InitializeParticle(
    Color=$User.Emissive Color
    Lifetime=0.8
    Position Offset=<0 20 0>
    Sprite Size=<1000 40>
    Sprite Size Mode=::NewEnumerator4
    UsePositionOffset=true
    Write Material Random=true
   )
   &CameraOffset(Camera Offset Amount=3)
  ]
  particleUpdateScript [
   &SkeletalMeshLocation(MaskSpawnGroup=false, Mesh Sampling Type=::NewEnumerator5, OverrideRandomSeed=false, Skeletal Mesh=$User.SkeletalMesh)
   &ParticleState()
   &ScaleColor(
    Scale RGB=&VectorFromFloat(Value=&FloatFromCurve(FloatCurve=DataInterface() {propertyValues {Curve {keys [<at 0 5 RCIM_Cubic>,<at 1 0 RCIM_Cubic>]}}}))
    ScaleA=true
    ScaleRGB=true
   )
   &DynamicMaterialParameters(
    Index 0 Param 1=&FloatFromCurve(FloatCurve=DataInterface() {propertyValues {Curve {keys [<at 0 0.9>,<at 0.5 0.05>]}}}, Scale Curve=0.5)
    Index 0 Param 2=0.25
    Index 0 Param 3=1
    Index 0 Param 4=1
    Module.Param0WriteEnabled=true
    Module.Param1WriteEnabled=true
    Module.Param2WriteEnabled=true
    Module.Param3WriteEnabled=true
   )
  ]
  rendererClasses [&NiagaraSpriteRendererProperties]
  renderers [{rendererClass=&NiagaraSpriteRendererProperties}]
 }
}
summary {
 systemName=NS_Glider_Alfredo_Male_Deploy
 emitters [
  {emitterName=Round_Flare, rendererClasses [&NiagaraSpriteRendererProperties]}
  {emitterName=Wide_Flare, rendererClasses [&NiagaraSpriteRendererProperties]}
  {emitterName=Beams, rendererClasses [&NiagaraSpriteRendererProperties]}
 ]
 userVariables [
  User.Emissive Color {type=LinearColor, description="", defaultValue {struct=LinearColor, value=<rgba 1 1 1 1>}}
  User.SkeletalMesh {
   type=NiagaraDataInterfaceSkeletalMesh
   defaultValue {
    struct=NiagaraExt_VariableValue_DataInterface
    value {dataInterface=&self:NiagaraDataInterfaceSkeletalMesh_0}
   }
  }
 ]
}
